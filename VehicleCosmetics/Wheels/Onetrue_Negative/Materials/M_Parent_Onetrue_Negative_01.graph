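# Parent material for the Onetrue Negative vehicle wheel, painting an animated aurora night sky across rim and tire.
# Two panning, curl-noise-distorted BumpOffset aurora layers take gradient or Painted-switch color, then combine with parallax stars, sky gradient and fresnel.
# Vertex red selects rim or tire normals and detail masks, which set tread color and roughness, cut the glow and carry scrolling emissive streaks.
# Output passes through shared vehicle damage and de-rez transition functions.
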Material /VehicleCosmetics/Wheels/Onetrue_Negative/Materials/M_Parent_Onetrue_Negative_01
graph {
 properties {
  bAutomaticallySetUsageInEditor=false
  bUseMaterialAttributes=true
  bUsedWithMeshDeformer=true
  bUsedWithSkeletalMesh=true
  usedShadingModels=Default Lit
 }
 nodes [
  n1 = TextureSampleParameter2D(UVs=Panner(Coordinate=TextureCoordinate() {uTiling=0.01}, Time=Time()) {speedX=0.05}) {
   parameterName=Onetrue_Negative_ColorGradient
   samplerType=SAMPLERTYPE_LinearColor
   texture=/VehicleCosmetics/Wheels/_SharedTextures/FX/T_Onetrue_Negative_ColorGradient
  }
  n2 = StaticSwitchParameter(True=Multiply(A=VectorParameter() {parameterName=Aurora_PaintedColor, defaultValue=<rgba 0.01 0.87 0.24 1>}.RGB, B=@n1.A), False=@n1.RGB) {parameterName=Painted}
  n3 = TextureCoordinate() {uTiling=4, vTiling=0.25}
  n4 = Multiply(A=Time(), B=ScalarParameter() {defaultValue=1, parameterName=Pan_Speed})
  n5 = Add(A=ScalarParameter() {defaultValue=0.25, parameterName=Aurora_TimeOffset}, B=@n4)
  n6 = Add(
   A=ScalarParameter() {defaultValue=-1.5, parameterName=Aurora2_LayerOffset}
   B=TextureSample(UVs=Panner(Time=@n5) {speedX=0.1}) {texture=/VehicleCosmetics/Wheels/_SharedTextures/FX/T_Onetrue_Negative_NoiseRGB}.A
  )
  n7 = ScalarParameter() {defaultValue=0.3, parameterName=Aurora_HeightRatio}
  n8 = BumpOffset(Coordinate=Panner(Coordinate=@n3, Time=@n5) {speedX=0.1}, Height=@n6, HeightRatioInput=@n7)
  n9 = Panner(Coordinate=@n3, Time=@n5) {speedX=0.2}
  n10 = ConstantBiasScale(None=TextureSample(UVs=BumpOffset(Coordinate=@n9, Height=@n6, HeightRatioInput=@n7)) {texture=/VehicleCosmetics/Wheels/_SharedTextures/FX/Universal_Noise_HeightMap_08}.RGB) {bias=0, scale=5}
  n11 = TextureCoordinate() {vTiling=0.8}
  n12 = Constant() {r=0.16}
  n13 = Panner(Coordinate=@n11, Time=@n12) {speedY=1}
  n14 = TextureCoordinate() {uTiling=0.5}
  n15 = Panner(Coordinate=@n14, Time=@n4) {speedX=0.2}
  n16 = Panner(Coordinate=@n14, Time=@n5) {speedX=0.05, speedY=0.05}
  n17 = Multiply(
   A=Add(
    A=TextureSample(UVs=@n15) {texture=/VehicleCosmetics/Wheels/_SharedTextures/FX/UVCurlNoise}.G
    B=TextureSample(UVs=@n16) {texture=&UVCurlNoise}.G
   )
   B=Constant() {r=0.15}
  )
  n18 = TextureSample(UVs=BumpOffset(Coordinate=Subtract(A=@n13, B=@n17), Height=@n6, HeightRatioInput=@n7)) {samplerType=SAMPLERTYPE_LinearColor, texture=/VehicleCosmetics/Wheels/Onetrue_Negative/Textures/T_Onetrue_Negative_Tire_GradientRGB1}
  n19 = Multiply(A=Multiply(A=TextureSample(UVs=@n8) {texture=/VehicleCosmetics/Wheels/_SharedTextures/FX/NoiseCone02_T}.RGB, B=Clamp(None=@n10)), B=@n18.R)
  n20 = Constant() {r=0.4}
  n21 = Multiply(A=Clamp(None=ConstantBiasScale(None=@n18.R) {bias=-0.1, scale=0.8}), B=@n20)
  n22 = Panner(Coordinate=@n3, Time=@n4) {speedX=0.1}
  n23 = TextureSample(UVs=Panner() {speedX=0.1}) {texture=&T_Onetrue_Negative_NoiseRGB}
  n24 = Divide(A=@n3, B=Constant() {r=2})
  n25 = TextureSample(UVs=BumpOffset(Coordinate=Panner(Coordinate=@n24, Time=@n4) {speedX=0.1}, Height=@n23.A, HeightRatioInput=@n7)) {texture=&T_Onetrue_Negative_NoiseRGB}
  n26 = Multiply(
   A=TextureSample(UVs=BumpOffset(Coordinate=@n22, Height=@n23.A)) {texture=&T_Onetrue_Negative_NoiseRGB}.B
   B=Clamp(None=ConstantBiasScale(None=@n25.G) {bias=0, scale=5})
  )
  n27 = Panner(Coordinate=@n11, Time=@n12, Speed=@n7) {speedY=1}
  n28 = TextureCoordinate()
  n29 = Panner(Coordinate=@n28) {speedX=0.2}
  n30 = Panner(Coordinate=@n28) {speedX=0.05, speedY=0.05}
  n31 = Multiply(
   A=Add(
    A=TextureSample(UVs=@n29) {texture=&UVCurlNoise}.G
    B=TextureSample(UVs=@n30) {texture=&UVCurlNoise}.G
   )
   B=Constant() {r=0.15}
  )
  n32 = TextureSample(UVs=BumpOffset(Coordinate=Subtract(A=@n27, B=@n31), Height=@n23.A)) {samplerType=SAMPLERTYPE_LinearColor, texture=&T_Onetrue_Negative_Tire_GradientRGB1}
  n33 = Clamp(None=ConstantBiasScale(None=@n32.R) {bias=-0.1, scale=0.8})
  n34 = Add(
   A=Multiply(A=Add(A=@n19, B=@n21), B=ScalarParameter() {defaultValue=0.25, parameterName=Aurora2Fade})
   B=Add(A=Multiply(A=@n26, B=@n32.R), B=Multiply(A=@n20, B=@n33))
  )
  n35 = Multiply(A=@n2, B=@n34)
  n36 = Multiply(A=Clamp(None=ConstantBiasScale(None=@n34) {bias=-0.2, scale=1}), B=ScalarParameter() {defaultValue=0.5, parameterName=HighLightIntensity})
  n37 = TextureCoordinate() {coordinateIndex=2}
  n38 = VertexColor()
  n39 = LinearInterpolate(
   A=Multiply(
    A=TextureSample(UVs=@n37) {samplerType=SAMPLERTYPE_LinearColor, texture=/VehicleCosmetics/Wheels/Onetrue_Negative/Textures/T_Onetrue_Negative_Rim_RGB}.R
    B=ScalarParameter() {defaultValue=1, parameterName=RimDetailOpacity}
   )
   B=Multiply(
    A=TextureSample() {texture=/VehicleCosmetics/Wheels/Onetrue_Negative/Textures/T_Onetrue_Negative_Tire_RGB}.R
    B=ScalarParameter() {defaultValue=1, parameterName=TreadDetailOpacity}
   )
   Alpha=@n38.R
  )
  n40 = TextureSample(UVs=Panner() {speedX=0.1, speedY=0.1}) {texture=&UVCurlNoise}
  n41 = TextureCoordinate()
  n42 = ScalarParameter() {defaultValue=0.5, parameterName=StarsTiling}
  n43 = ScalarParameter() {defaultValue=-10, parameterName=SkyOffset}
  n44 = Panner(Coordinate=BumpOffset(Coordinate=Multiply(A=@n41, B=@n42), Height=@n43)) {speedY=-0.015}
  n45 = Multiply(
   A=Multiply(
    A=ConstantBiasScale(None=@n40.RGB) {bias=0.2, scale=1}
    B=TextureSample(UVs=@n44) {texture=/VehicleCosmetics/Wheels/_SharedTextures/FX/Noise_Spots_RGB_T}.B
   )
   B=ScalarParameter() {defaultValue=3, parameterName=StarsIntensity}
  )
  n46 = TextureSample(UVs=BumpOffset(Coordinate=@n41, Height=@n43)) {samplerType=SAMPLERTYPE_LinearColor, texture=&T_Onetrue_Negative_Tire_GradientRGB1}
  n47 = LinearInterpolate(
   A=Multiply(
    A=ScalarParameter() {defaultValue=1, parameterName=SkyDarIntensity}
    B=VectorParameter() {parameterName=SkyDarkColor, defaultValue=<rgba 0.02 0.06 0.2 1>}.RGB
   )
   B=Multiply(
    A=ScalarParameter() {defaultValue=1, parameterName=SkyLightIntensity}
    B=VectorParameter() {parameterName=SkyLightColor, defaultValue=<rgba 0.14 0.14 0.32 1>}.RGB
   )
   Alpha=@n46.G
  )
  n48 = TextureSampleParameter2D() {parameterName=Rim_N, samplerType=SAMPLERTYPE_Normal, texture=/VehicleCosmetics/Wheels/Onetrue_Negative/Textures/OneTrue_negative_rim_N}
  n49 = TextureSampleParameter2D() {parameterName=Tire_N, samplerType=SAMPLERTYPE_Normal, texture=/VehicleCosmetics/Wheels/Onetrue_Negative/Textures/T_Onetrue_Negative_Tire_N}
  n50 = TextureSample(UVs=TextureCoordinate()) {samplerType=SAMPLERTYPE_Normal, texture=/VehicleCosmetics/Wheels/Onetrue_Negative/Textures/Matte_N}
  n51 = VectorParameter() {parameterName=TireDetailNormalIntensity, defaultValue=<rgba 0.38 0.38 0 1>}
  n52 = LinearInterpolate(A=@n48.RGB, B=Add(A=@n49.RGB, B=Multiply(A=@n50.RGB, B=@n51.RGB)), Alpha=@n38.R)
  n53 = ScalarParameter() {defaultValue=3, parameterName=Fresnel_Sharpness}
  n54 = Multiply(
   A=@n2
   B=MaterialFunctionCall!(Input Normal=@n52, Input Sharpness=@n53, Input Two Sided) {materialFunction=/VehicleCosmetics/Wheels/CarCar_Holographic/Materials/MF_Fresnel_Advanced}.Edge
  )
  n55 = ScalarParameter() {defaultValue=2, parameterName=AuroraIntensity}
  n56 = Add(A=Add(A=Add(A=@n45, B=@n47), B=Multiply(A=@n54, B=@n53)), B=Add(A=Multiply(A=@n35, B=@n55), B=@n36))
  n57 = ScalarParameter() {parameterName=TreadDetailIntensity}
  n58 = TextureSample(UVs=Panner(Coordinate=TextureCoordinate() {coordinateIndex=2, uTiling=2}) {speedX=-0.25}) {samplerType=SAMPLERTYPE_LinearColor, texture=&T_Onetrue_Negative_Tire_GradientRGB1}
  n59 = Multiply(A=Multiply(A=@n58.B, B=@n39), B=VectorParameter() {parameterName=StreaksColor, defaultValue=<rgba 0.12 0 0.86 1>}.RGB)
  n60 = ScalarParameter() {defaultValue=3, parameterName=StreaksIntensity}
  n61 = ScalarParameter() {defaultValue=1, parameterName=Roughness}
  n62 = LinearInterpolate(A=Constant() {r=1}, B=@n2, Alpha=Constant())
  n63 = SetMaterialAttributes!(
   MaterialAttributes
   Base Color=LinearInterpolate(A=Add(A=@n35, B=@n36), B=VectorParameter() {parameterName=TreadColor, defaultValue=<rgba 0.06 0 0.42 1>}.RGB, Alpha=@n39)
   Emissive Color=Add(A=LinearInterpolate(A=@n56, B=@n57, Alpha=@n39), B=Multiply(A=@n59, B=@n60))
   Roughness=LinearInterpolate(A=Multiply(A=@n61, B=@n62), B=ScalarParameter() {defaultValue=0.3, parameterName=Roughness2}, Alpha=OneMinus(None=@n39))
   Normal=@n52
   Metallic=ScalarParameter() {parameterName=Metallic}
  ) {attributeSetTypes [
    #1
    #2
    #3
    #4
    #5
   ]}
  MaterialFunctionCall(In=MaterialFunctionCall(Material=@n63) {materialFunction=&MF_BaseVehicle_Damage}.Result) {materialFunction=&MF_VehicleDeRez_Transition}
  Multiply()
 ]
}
