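# Parent material for the OneTrue Negative vehicle wheel cosmetic, adding animated emissive effects to the simple wheel parent function.
# A rim RGB mask separates scrolling streaks and lanes, a panning aurora gradient with an optional Painted tint, and tintable highlights.
# Inverted red vertex color blends their sum over the base emissive, then the attributes pass through vehicle damage and de-rez transition functions.
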Material /VehicleCosmetics/Wheels/Onetrue_Negative/Materials/M_WheelParent_OneTrue_Negative
graph {
 properties {
  bAutomaticallySetUsageInEditor=false
  bUseMaterialAttributes=true
  bUsedWithMeshDeformer=true
  bUsedWithSkeletalMesh=true
  pixelDepthOffsetMode=PDOM_Legacy
  usedShadingModels=Default Lit
 }
 nodes [
  n1 = MaterialFunctionCall() {materialFunction=&MF_WheelParent_Simple}
  n2 = GetMaterialAttributes(None=@n1.MaterialAttributes) {attributeGetTypes [#1]}
  n3 = LinearInterpolate(A=@n2.Emissive Color, B=NamedRerouteUsage(), Alpha=OneMinus(None=ComponentMask(None=@n1.VertexColorR) {r=true}))
  MaterialFunctionCall(In=MaterialFunctionCall(Material=SetMaterialAttributes(MaterialAttributes=@n2.MaterialAttributes, Emissive Color=@n3) {attributeSetTypes [#1]}) {materialFunction=&MF_BaseVehicle_Damage}.Result) {materialFunction=&MF_VehicleDeRez_Transition}
  n4 = TextureSample(UVs=Panner(Coordinate=TextureCoordinate() {coordinateIndex=2}) {speedX=-0.25}) {samplerType=SAMPLERTYPE_LinearColor, texture=/VehicleCosmetics/Wheels/Onetrue_Negative/Textures/T_Onetrue_Negative_Tire_GradientRGB1}
  n5 = TextureSample(UVs=TextureCoordinate() {coordinateIndex=2}) {samplerType=SAMPLERTYPE_LinearColor, texture=/VehicleCosmetics/Wheels/Onetrue_Negative/Textures/T_Onetrue_Negative_Rim_RGB}
  n6 = Multiply(A=@n4.B, B=@n5.R)
  n7 = VectorParameter() {parameterName=Streaks_Color, defaultValue=<rgba 0.33 0.24 0.86 1>}
  n8 = Add(
   A=Multiply(A=Multiply(A=@n6, B=@n7.RGB), B=ScalarParameter() {defaultValue=3, parameterName=Streaks_Intensity})
   B=Multiply(A=Multiply(A=@n5.R, B=@n7.RGB), B=ScalarParameter() {defaultValue=0.5, parameterName=StreaksLanes_Intensity})
  )
  n9 = VectorParameter() {parameterName=Aurora_PaintColor, defaultValue=<rgba 0.01 0.87 0.24 1>}
  n10 = TextureSample(UVs=Panner(Coordinate=TextureCoordinate() {uTiling=0.01}, Time=Time()) {speedX=0.05}) {samplerType=SAMPLERTYPE_LinearColor, texture=/VehicleCosmetics/Wheels/_SharedTextures/FX/T_Onetrue_Negative_ColorGradient}
  n11 = Multiply(
   A=StaticSwitchParameter(True=Multiply(A=@n9.RGB, B=@n10.A), False=@n10.RGB) {parameterName=Painted}
   B=ScalarParameter() {defaultValue=1, parameterName=Aurora_IntensityColor}
  )
  n12 = Multiply(A=@n5.G, B=@n11)
  n13 = Multiply(A=LinearInterpolate(A=Constant() {r=1}, B=@n11, Alpha=ScalarParameter() {parameterName=HighlightTint}), B=@n5.B)
  n14 = ScalarParameter() {defaultValue=0.8, parameterName=Highlight_Intensity}
  NamedRerouteDeclaration(None=Add(A=Add(A=@n8, B=@n12), B=Multiply(A=@n13, B=@n14))) {name=Emissive, nodeColor=<rgba 1 0.75 0 1>}
 ]
}
